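# Material for the base gem of the Piper Shelf medallion in the Blitz reveal gadget, using Default Lit and material attributes.
# It wraps the shared 2023 character parent PreFX and PostFX functions around an optional gem layer.
# A "Piper Shelf" static switch routes the PreFX result and Tech Art Mask through MF_NewDiamond_Medallion_Blitz, or passes the result straight to PostFX.
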Material /BRCosmetics/Gadgets/Assets/Medallion_BlitzReveal/Materials/M_Medallion_PiperShelf_BaseGem
graph {
 properties {bAutomaticallySetUsageInEditor=false, bUseMaterialAttributes=true, usedShadingModels=Default Lit}
 nodes [
  n1 = MaterialFunctionCall() {materialFunction=&MF_CharacterParent_2023_PreFX}
  MaterialFunctionCall(
   Material=StaticSwitchParameter(
    True=MaterialFunctionCall!(Material=@n1.Result, TechArtMask=@n1.Tech Art Mask) {materialFunction=&/MF_NewDiamond_Medallion_Blitz}.Material
    False=@n1.Result
   ) {parameterName=Piper Shelf}
   M=@n1.M
   Pre Skinned Local Position=@n1.M
   Distance Blend=@n1.M
   Fresnel=@n1.M
  ) {materialFunction=&MF_CharacterParent_2023_PostFX}
 ]
}
